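# Animation Blueprint for the single-shot flintlock pistol weapon mesh, adjusting its pose by the wielding player's gender.
# A non-looping idle runs through DefaultSlot into a cached pose.
# An enum blend outputs that pose either directly or through the FemaleAdditive slot.
# Each update, the event graph casts the owning actor to B_Pistol_Flintlock_Athena and copies its instigator pawn's character gender into the Gender variable driving the blend.
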AnimBlueprint /Game/Animation/Game/MainPlayer/Combat/Guns/FlintLock_Pistol/AnimBP_FlintLock_Pistol
target_skeleton=/Game/Weapons/FORT_Pistols/Mesh/SK_Flintlock_Pistol_Single_Shot_Skeleton
graphs [
 AnimGraph {nodes [
   n1 = "Blend Poses (EFortCustomGender)"!(
    BlendPose_0:Pose Link Structure=Use cached pose 'Pose'()
    BlendPose_1:Pose Link Structure=Slot 'FemaleAdditive'(Source=Use cached pose 'Pose'()) {slotName=FemaleAdditive}
    BlendTime_0:float=0.1
    BlendTime_1:float=0.1
    ActiveEnumValue:Byte=Get Gender() -> (Gender:EFortCustomGender Enum=Invalid)
   ) -> (Pose:Pose Link Structure) {
    is=title AnimGraphNode_BlendListByEnum
    blendType=Linear
    blendPose [{linkId=-1, sourceLinkId=-1}, {linkId=-1, sourceLinkId=-1}]
    blendTime [0.1,0.1]
   }
   Output Pose(Result=@n1) {layerGroup=None}
   Save cached pose 'Pose'(Pose=Slot 'DefaultSlot'(Source=Sequence Player 'FlintLockPistol_Idle_CMM_Weapon'() {bLoopAnimation=false, sequence=&/FlintLockPistol_Idle_CMM_Weapon})) {cacheName=Pose}
  ]}
 EventGraph {nodes [
   Event BlueprintUpdateAnimation()
   > n1 = Cast To B_Pistol_Flintlock_Athena!(execute, Object=GetOwningActor()) -> (then, CastFailed, AsB Pistol Flintlock Athena:B Pistol Flintlock Athena Object Reference) {is=title K2Node_DynamicCast}
   > n2 = Cast To FortPlayerPawnAthena(Object=Get Instigator!(self:Actor Object Reference=@n1) -> (Instigator:Pawn Object Reference))
   > Set Gender(Gender:EFortCustomGender Enum=GetCharacterGender!(self:Fort Player Pawn Object Reference=@n2) -> (ReturnValue:EFortCustomGender Enum=Invalid)) -> (Output_Get:EFortCustomGender Enum=Invalid)
  ]}
]
variables [Gender {type=EFortCustomGender, default=Invalid}]
