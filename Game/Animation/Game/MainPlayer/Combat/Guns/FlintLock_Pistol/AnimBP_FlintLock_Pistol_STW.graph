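# Animation Blueprint for the single-shot flintlock pistol's weapon skeleton, in the main player's gun combat animation set.
# A non-looping idle sequence passes through DefaultSlot for montage overrides and is cached as the base pose.
# A blend by the Gender enum outputs either that pose or the same pose run through the FemaleAdditive slot.
# The event graph holds only an empty update event, so nothing here sets Gender from its Invalid default.
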AnimBlueprint /Game/Animation/Game/MainPlayer/Combat/Guns/FlintLock_Pistol/AnimBP_FlintLock_Pistol_STW
target_skeleton=/Game/Weapons/FORT_Pistols/Mesh/SK_Flintlock_Pistol_Single_Shot_Skeleton
graphs [
 AnimGraph {nodes [
   n1 = "Blend Poses (EFortCustomGender)"!(
    BlendPose_0:Pose Link Structure=Use cached pose 'Pose'()
    BlendPose_1:Pose Link Structure=Slot 'FemaleAdditive'(Source=Use cached pose 'Pose'()) {slotName=FemaleAdditive}
    BlendTime_0:float=0.1
    BlendTime_1:float=0.1
    ActiveEnumValue:Byte=Get Gender() -> (Gender:EFortCustomGender Enum=Invalid)
   ) -> (Pose:Pose Link Structure) {
    is=title AnimGraphNode_BlendListByEnum
    blendType=Linear
    blendPose [{linkId=-1, sourceLinkId=-1}, {linkId=-1, sourceLinkId=-1}]
    blendTime [0.1,0.1]
   }
   Output Pose(Result=@n1) {layerGroup=None}
   Save cached pose 'Pose'(Pose=Slot 'DefaultSlot'(Source=Sequence Player 'FlintLockPistol_Idle_CMM_Weapon'() {bLoopAnimation=false, sequence=&/FlintLockPistol_Idle_CMM_Weapon})) {cacheName=Pose}
  ]}
 EventGraph {nodes [Event BlueprintUpdateAnimation()]}
]
variables [Gender {type=EFortCustomGender, default=Invalid}]
